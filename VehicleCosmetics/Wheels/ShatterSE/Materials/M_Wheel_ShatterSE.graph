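# Masked material for the ShatterSE vehicle wheel, adding animated, tintable glows to a near-black base color.
# It overrides MF_WheelParent_Simple's base color and emissive, then passes through damage and de-rez transition functions.
# Panning gradients drive masked rim pulses and tread waves, with 180-degree-rotated UVs reversing part of the tread flow.
# Named reroutes carry these pulses, two custom colors and a fresnel term using a blended rim/tread normal.
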Material /VehicleCosmetics/Wheels/ShatterSE/Materials/M_Wheel_ShatterSE
graph {
 properties {
  bAutomaticallySetUsageInEditor=false
  bUseMaterialAttributes=true
  bUsedWithMeshDeformer=true
  bUsedWithSkeletalMesh=true
  blendMode=BLEND_Masked
  pixelDepthOffsetMode=PDOM_Legacy
  previewMesh=/VehicleCosmetics/Wheels/ShatterSE/Meshes/SK_Wheel_ShatterSE
  usedShadingModels=Default Lit
 }
 nodes [
  n1 = MaterialFunctionCall() {materialFunction=&MF_WheelParent_Simple}
  n2 = GetMaterialAttributes(None=@n1.MaterialAttributes) {attributeGetTypes [#1, #2]}
  n3 = NamedRerouteUsage()
  n4 = ScalarParameter() {defaultValue=0.01, group=5_Special Effects, parameterName=BaseColor}
  MaterialFunctionCall(In=MaterialFunctionCall(Material=SetMaterialAttributes!(MaterialAttributes=@n2.MaterialAttributes, Emissive Color=@n3, Base Color=@n4) {attributeSetTypes [#3, #2]}) {materialFunction=&MF_BaseVehicle_Damage}.Result) {materialFunction=&MF_VehicleDeRez_Transition}
  n5 = NamedRerouteUsage()
  n6 = NamedRerouteUsage()
  n7 = NamedRerouteUsage()
  n8 = NamedRerouteUsage()
  n9 = Add(
   A=Multiply(A=OneMinus(None=@n5), B=@n6)
   B=Multiply(A=Multiply(A=@n7, B=@n8), B=ScalarParameter() {defaultValue=0.75, group=5_Special Effects, parameterName=LineRim Intensity, sortPriority=4})
  )
  n10 = TextureSampleParameter2D() {
   group=5_Special Effects
   parameterName=Mask_Rim
   samplerType=SAMPLERTYPE_LinearColor
   sortPriority=1
   texture=/VehicleCosmetics/Wheels/ShatterSE/Textures/T_Rim_ShatterSE_Mask
  }
  n11 = TextureSample(UVs=Panner(Coordinate=TextureCoordinate() {coordinateIndex=1}) {speedX=-0.5}, Tex=NamedRerouteUsage()) {samplerSource=SSM_Wrap_WorldGroupSettings, texture=/VehicleCosmetics/Wheels/CarCar_Holographic/Textures/Gradiant_V}
  n12 = Saturate(None=Subtract(A=@n10.RGB, B=Multiply(A=@n11.R) {constB=2}))
  n13 = Desaturation(
   None=Multiply(A=Multiply(A=@n12, B=@n6)) {constB=10}
   Fraction=ScalarParameter() {defaultValue=0.125, group=5_Special Effects, parameterName=BlinkRim Desaturation, sortPriority=5}
  )
  n14 = Multiply(A=@n11.R, B=OneMinus(None=NamedRerouteUsage()))
  n15 = NamedRerouteUsage()
  n16 = Power(
   Base=Multiply(
    A=Multiply(A=@n14, B=@n15)
    B=ScalarParameter() {defaultValue=1.125, group=5_Special Effects, parameterName=Rim Pulse Intensity, sortPriority=6}
   )
   Exp=ScalarParameter() {defaultValue=3, group=5_Special Effects, parameterName=Rim Pulse Contrast, sortPriority=7}
  )
  n17 = Multiply(A=@n16, B=NamedRerouteUsage())
  n18 = ScalarParameter() {defaultValue=0.825, group=5_Special Effects, parameterName=Rim Pulse Desaturate, sortPriority=8}
  NamedRerouteDeclaration(None=Add(A=Add(A=@n9, B=@n13) {constB=0}, B=LinearInterpolate(A=@n16, B=@n17, Alpha=@n18) {constAlpha=0})) {name=Rim Pulse, nodeColor=<rgba 0.5 0 1 1>}
  n19 = TextureSampleParameter2D() {
   group=5_Special Effects
   parameterName=Normal Tread
   samplerType=SAMPLERTYPE_Normal
   sortPriority=3
   texture=/VehicleCosmetics/Wheels/Shatter/Textures/T_Tire_Shatter_N
  }
  n20 = LinearInterpolate(A=NamedRerouteUsage(), B=@n19.RGB, Alpha=NamedRerouteUsage())
  n21 = Multiply(
   A=MaterialFunctionCall!(
    Input Normal=NamedRerouteDeclaration(None=@n20) {name=Normal_output, nodeColor=<rgba 0.5 0 1 1>}
    Input Sharpness=ScalarParameter() {defaultValue=1.5, group=5_Special Effects, parameterName=Fresnel Sharpness, sortPriority=11}
    Input Two Sided
   ) {materialFunction=/VehicleCosmetics/Wheels/CarCar_Holographic/Materials/MF_Fresnel_Advanced}.Edge
   B=ScalarParameter() {defaultValue=0.5, group=5_Special Effects, parameterName=Fresnel Intensity, sortPriority=12}
  )
  n22 = TextureSample(Tex=NamedRerouteUsage()) {samplerType=SAMPLERTYPE_LinearColor, texture=/VehicleCosmetics/Wheels/ShatterSE/Textures/T_Tire_ShatterSE_RGB}
  n23 = LinearInterpolate(A=NamedRerouteUsage(), B=LinearInterpolate(A=NamedRerouteUsage(), B=NamedRerouteUsage(), Alpha=@n22.A), Alpha=NamedRerouteUsage())
  NamedRerouteDeclaration(None=LinearInterpolate(
    A=@n21
    B=Multiply(A=@n21, B=@n23)
    Alpha=ScalarParameter() {defaultValue=1, group=5_Special Effects, parameterName=Fresnel Desaturate, sortPriority=13}
   ) {constAlpha=0}) {name=Fresnel, nodeColor=<rgba 0.5 0 1 1>}
  n24 = NamedRerouteUsage()
  n25 = NamedRerouteUsage()
  n26 = NamedRerouteUsage()
  NamedRerouteDeclaration(None=Add(A=LinearInterpolate(A=@n24, B=@n25, Alpha=@n26), B=NamedRerouteUsage())) {name=Emission_output, nodeColor=<rgba 0.5 0 1 1>}
  n27 = MaterialFunctionCall(Color=@n1.RimRGBAMask) {materialFunction=&SplitComponents}
  NamedRerouteDeclaration(None=@n27.R) {name=RimRMask, nodeColor=<rgba 0.5 0 1 1>}
  NamedRerouteDeclaration(None=ComponentMask(None=@n1.RimRGBAMask) {a=true}) {name=RimAMask, nodeColor=<rgba 0.5 0 1 1>}
  NamedRerouteDeclaration(None=@n27.G) {name=RimGMask, nodeColor=<rgba 0.5 0 1 1>}
  NamedRerouteDeclaration(None=@n27.B) {name=RimBMask, nodeColor=<rgba 0.5 0 1 1>}
  n28 = VertexColor()
  NamedRerouteDeclaration(None=Multiply(A=@n2.Normal, B=OneMinus(None=@n28.R))) {name=RimNormal_input, nodeColor=<rgba 0.5 0 1 1>}
  n29 = VectorParameter() {group=5_Special Effects, parameterName=CustomColor2, sortPriority=0, defaultValue=<rgba 0 0.255 0.0405 1>}
  NamedRerouteDeclaration(None=@n29.RGB) {name=AccentColor, nodeColor=<rgba 0.5 0 1 1>}
  NamedRerouteDeclaration(None=VertexColor().R) {name=VC_R, nodeColor=<rgba 0.5 0 1 1>}
  n30 = TextureObjectParameter() {
   group=5_Special Effects
   parameterName=Gradient
   samplerType=SAMPLERTYPE_LinearColor
   sortPriority=2
   texture=&T_Tire_ShatterSE_RGB
  }
  NamedRerouteDeclaration(None=@n30) {name=TireRGB_Input, nodeColor=<rgba 0.5 0 1 1>}
  n31 = VectorParameter() {group=5_Special Effects, parameterName=CustomColor1, sortPriority=1, defaultValue=<rgba 0.5 1 0 1>}
  NamedRerouteDeclaration(None=@n31.RGB) {name=Custom Color, nodeColor=<rgba 0.5 0 1 1>}
  n32 = MaterialFunctionCall(UVs=TextureCoordinate(), "Rotation Angle (0-1)"=Constant() {r=0.5}) {materialFunction=&CustomRotator}
  n33 = NamedRerouteUsage()
  n34 = Multiply(
   A=OneMinus(None=@n22.B)
   B=TextureSample(UVs=Panner(Coordinate=@n32.Rotated Values) {speedX=-0.5}, Tex=@n33) {samplerSource=SSM_Wrap_WorldGroupSettings, texture=&Gradiant_V}.R
  )
  n35 = Multiply(
   A=@n22.B
   B=TextureSample(UVs=Panner() {speedX=-0.5}, Tex=@n33) {samplerSource=SSM_Wrap_WorldGroupSettings, texture=&Gradiant_V}.R
  )
  n36 = Multiply(
   A=@n22.G
   B=Multiply(A=Add(A=@n34, B=@n35), B=ScalarParameter() {defaultValue=5, group=5_Special Effects, parameterName=Wave Glow Intensity, sortPriority=9})
  )
  NamedRerouteDeclaration(None=LinearInterpolate(
    A=@n36
    B=Multiply(A=@n36, B=@n23)
    Alpha=ScalarParameter() {defaultValue=0.95, group=5_Special Effects, parameterName=Wave Glow Desaturate, sortPriority=10}
   ) {constAlpha=0}) {name=Tread Pulse, nodeColor=<rgba 0.5 0 1 1>}
  TextureSampleParameter2D() {
   group=5_Special Effects
   parameterName=Mask_Rim
   samplerType=SAMPLERTYPE_LinearColor
   sortPriority=1
   texture=&T_Rim_ShatterSE_Mask
  }
  TextureSampleParameter2D() {
   group=5_Special Effects
   parameterName=Normal Tread
   samplerType=SAMPLERTYPE_Normal
   sortPriority=3
   texture=&T_Tire_Shatter_N
  }
 ]
}
